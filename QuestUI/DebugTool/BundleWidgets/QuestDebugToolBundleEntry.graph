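# Quest debug tool row widget that shows one challenge bundle quest entry: definition, display data, name, unlock info, progress, state and progress-quest flag.
# Set Data stores the entry struct in ActiveEntry with a field-notify broadcast; the event graph is empty.
# Helper functions format asset names, enums and progress as text, with fallbacks for missing assets or zero required count.
# Inactive quests get half content opacity and a different pip color.
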WidgetBlueprint /QuestUI/DebugTool/BundleWidgets/QuestDebugToolBundleEntry
graphs [
 EventGraph {links [], nodes []}
 Get Asset Name {nodes [
   n1 = Get Asset Name!() -> (then, Asset:Primary Data Asset Object Reference) {is=title K2Node_FunctionEntry}
   > n2 = Is Valid(InputObject=@n1)
   Return Node!(execute=@n2.Is Not Valid, AssetName:Text="LOCTEXT(\"[Not found]\")")
   > Return Node(AssetName="To Text (String)"(InString=GetObjectName(Object=@n1)))
  ]}
 Get Content Opacity {nodes [
   n1 = Get Content Opacity!() -> (then, Entry:"Bundle Quest Entry Debug Data Structure (by ref)") {is=title K2Node_FunctionEntry}
   n2 = Break Bundle Quest Entry Debug Data!(BundleQuestEntryDebugData:"Bundle Quest Entry Debug Data Structure (by ref)"=@n1) -> (QuestListItemId:String, QuestState:EFortQuestState Enum, bIsProgressQuest:bool=~) {is=title K2Node_BreakStruct}
   > Return Node!(
    execute
    Opacity:double=SelectFloat(A=0.5, B=1, bPickA="Equal (Enum)"(A:EFortQuestState Enum=@n2.QuestState, B:EFortQuestState Enum=Inactive))
   )
  ]}
 Get Entry Progress {nodes [
   n1 = Get Entry Progress!() -> (then, Entry:"Bundle Quest Entry Debug Data Structure (by ref)") {is=title K2Node_FunctionEntry}
   n2 = Break Bundle Quest Entry Debug Data!(BundleQuestEntryDebugData=@n1) -> (
    QuestDefinition:Fort Quest Item Definition Object Reference
    QuestDisplayData:Fort Quest Display Data Asset Object Reference
    QuestListItemId
    DisplayName:Text
    UnlockType:EChallengeBundleQuestUnlockType Enum
    UnlockValue:int
    QuestAchievedCount:int
    QuestRequiredCount:int
    bIsProgressQuest=~
   )
   > n3 = Branch(Condition="integer > integer"(A=@n2.QuestRequiredCount))
   > Return Node!(execute, Progress:Text=Format Text!(Format="LOCTEXT(\"{0} / {1}\")", 0:int=@n2.QuestAchievedCount, 1:int=@n2.QuestRequiredCount) -> (Result))
   Return Node(execute=@n3.else, Progress="LOCTEXT(\"N/A\")")
  ]}
 Get Entry State {nodes [
   n1 = Get Entry State!() -> (then, Entry:"Bundle Quest Entry Debug Data Structure (by ref)") {is=title K2Node_FunctionEntry}
   n2 = Break Bundle Quest Entry Debug Data!(BundleQuestEntryDebugData=@n1) -> (QuestListItemId, QuestState:EFortQuestState Enum, bIsProgressQuest=~)
   > Return Node!(execute, State:Text="To Text (String)"(InString=Enum to String!(Enumerator:EFortQuestState Enum=@n2.QuestState) -> (ReturnValue:String)))
  ]}
 Get Entry Unlock Type {nodes [
   n1 = Get Entry Unlock Type!() -> (then, Entry:"Bundle Quest Entry Debug Data Structure (by ref)") {is=title K2Node_FunctionEntry}
   > Return Node!(
    execute
    UnlockType:Text=To Text!(Value:EChallengeBundleQuestUnlockType Enum=Break Bundle Quest Entry Debug Data!(BundleQuestEntryDebugData=@n1) -> (UnlockType:EChallengeBundleQuestUnlockType Enum)) -> (ReturnValue:Text) {is=title K2Node_GenericToText}
   )
  ]}
 Get Is Progress Quest {nodes [
   n1 = Get Is Progress Quest!() -> (then, IsProgressQuest:bool=~) {is=title K2Node_FunctionEntry}
   > Return Node!(execute, Output:Text=SelectText(A="LOCTEXT(\"True\")", B="LOCTEXT(\"False\")", bPickA=@n1))
  ]}
 Get Pip Color {nodes [
   n1 = Get Pip Color!() -> (then, Entry:"Bundle Quest Entry Debug Data Structure (by ref)") {is=title K2Node_FunctionEntry}
   n2 = "Equal (Enum)"(A=Break Bundle Quest Entry Debug Data!(BundleQuestEntryDebugData=@n1) -> (QuestState:EFortQuestState Enum), B=Inactive)
   > Return Node!(
    execute
    Color:Linear Color Structure=Get Debug List Entry Type Color!(EntryType:EQuestPanelDebugListEntryType Enum=Select(Option 0:EQuestPanelDebugListEntryType Enum=QuestView, Option 1:EQuestPanelDebugListEntryType Enum=QuestAsset, Index=@n2) -> (ReturnValue:EQuestPanelDebugListEntryType Enum)) -> (Color:Linear Color Structure) {is=title K2Node_CallFunction}
   )
  ]}
 Set Data {nodes [
   n1 = Set Data!() -> (then, Entry:"Bundle Quest Entry Debug Data Structure (by ref)") {is=title K2Node_FunctionEntry}
   > Set with Broadcast ActiveEntry!(execute, ActiveEntry:Bundle Quest Entry Debug Data Structure=@n1) -> (then, Output_Get:Bundle Quest Entry Debug Data Structure)
  ]}
]
variables [ActiveEntry {
  type=BundleQuestEntryDebugData
  default {
   bIsProgressQuest=false
   displayName=""
   questAchievedCount=0
   questDefinition=null
   questDisplayData=null
   questListItemId=""
   questRequiredCount=0
   questState=Inactive
   unlockType=Manually
   unlockValue=0
  }
 }]
widgets {
 info {inheritedWidgetCount=0, parentClass=/Script/QuestUI.QuestPanelDebugToolWidgetBase, rootWidgetClass=&HorizontalBox, widgetCount=16}
 widgets [
  HorizontalBox_1 {widgetClassPath=&HorizontalBox, widget=HorizontalBox_1}
  Button_SelectEntry {
   widgetClassPath=/QuestUI/DebugTool/CommonWidgets/QuestDebugToolSelectEntryButton
   bIsVariable=true
   parent=HorizontalBox_1
   widget=Button_SelectEntry
   slot {class=HorizontalBoxSlot, properties {verticalAlignment=VAlign_Center, padding=<ltrb 4 4 4 4>}}
  }
  Pip {
   widgetClassPath=&Image
   bIsVariable=true
   parent=HorizontalBox_1
   widget=Pip
   properties {brush {
     bIsDynamicallyLoaded=false
     drawAs=Image
     imageType=NoImage
     mirroring=NoMirror
     resourceName=None
     resourceObject=null
     tiling=NoTile
     imageSize=<8 8>
     margin=<ltrb 0 0 0 0>
     outlineSettings {
      bUseBrushTransparency=false
      roundingType=HalfHeightRadius
      width=0
      color {colorUseRule=UseColor_Specified, specifiedColor=<rgba 0 0 0 0>}
      cornerRadii=<0 0 0 0>
     }
     tintColor {colorUseRule=UseColor_Specified, specifiedColor=<rgba 1 1 1 1>}
     uVRegion {bIsValid=false, max=<0 0>, min=<0 0>}
    }}
   slot {class=HorizontalBoxSlot, properties {padding=<ltrb 0 6 0 6>}}
  }
  Content {
   widgetClassPath=&GridPanel
   parent=HorizontalBox_1
   widget=Content
   properties {bOverride_Cursor=true, columnFill [5,3,1,1,1]}
   slot {class=HorizontalBoxSlot, properties {size {sizeRule=Fill, value=1}}}
  }
  HorizontalBox_0 {widgetClassPath=&HorizontalBox, parent=Content, widget=HorizontalBox_0, slot {class=GridSlot}}
  Button_DefinitionBrowse {
   widgetClassPath=/QuestUI/DebugTool/CommonWidgets/QuestDebugToolBrowseButton
   bIsVariable=true
   parent=HorizontalBox_0
   widget=Button_DefinitionBrowse
   slot {class=HorizontalBoxSlot, properties {verticalAlignment=VAlign_Center, padding=<ltrb 4 4 4 4>}}
  }
  Widget_DefinitionField {
   widgetClassPath=/QuestUI/DebugTool/CommonWidgets/QuestDebugToolInfoField
   bIsVariable=true
   parent=HorizontalBox_0
   widget=Widget_DefinitionField
   properties {header=Quest Definition}
   slot {class=HorizontalBoxSlot, properties {size {sizeRule=Fill, value=1}}}
  }
  HorizontalBox_2 {widgetClassPath=&HorizontalBox, parent=Content, widget=HorizontalBox_2, slot {class=GridSlot, properties {row=1}}}
  Button_DisplayDataBrowse {
   widgetClassPath=&QuestDebugToolBrowseButton
   bIsVariable=true
   parent=HorizontalBox_2
   widget=Button_DisplayDataBrowse
   slot {class=HorizontalBoxSlot, properties {verticalAlignment=VAlign_Center, padding=<ltrb 4 4 4 4>}}
  }
  Widget_DisplayDataField {
   widgetClassPath=&QuestDebugToolInfoField
   bIsVariable=true
   parent=HorizontalBox_2
   widget=Widget_DisplayDataField
   properties {header=Quest Display Data}
   slot {class=HorizontalBoxSlot, properties {size {sizeRule=Fill, value=1}}}
  }
  Widget_DisplayName {
   widgetClassPath=&QuestDebugToolInfoField
   bIsVariable=true
   parent=Content
   widget=Widget_DisplayName
   properties {header=Display Name, multiline=true}
   slot {class=GridSlot, properties {column=1, rowSpan=2}}
  }
  Widget_UnlockType {
   widgetClassPath=&QuestDebugToolInfoField
   bIsVariable=true
   parent=Content
   widget=Widget_UnlockType
   properties {header=Unlock Type}
   slot {class=GridSlot, properties {column=2, columnSpan=2}}
  }
  Widget_UnlockValue {
   widgetClassPath=&QuestDebugToolInfoField
   bIsVariable=true
   parent=Content
   widget=Widget_UnlockValue
   properties {header=Unlock Value}
   slot {class=GridSlot, properties {column=4}}
  }
  Widget_Progress {
   widgetClassPath=&QuestDebugToolInfoField
   bIsVariable=true
   parent=Content
   widget=Widget_Progress
   properties {header=Progress}
   slot {class=GridSlot, properties {column=2, row=1}}
  }
  Widget_State {
   widgetClassPath=&QuestDebugToolInfoField
   bIsVariable=true
   parent=Content
   widget=Widget_State
   properties {header=State}
   slot {class=GridSlot, properties {column=3, row=1}}
  }
  Widget_IsProgressQuest {
   widgetClassPath=&QuestDebugToolInfoField
   bIsVariable=true
   parent=Content
   widget=Widget_IsProgressQuest
   properties {header=Is Progress Quest}
   slot {class=GridSlot, properties {column=4, row=1}}
  }
 ]
}
